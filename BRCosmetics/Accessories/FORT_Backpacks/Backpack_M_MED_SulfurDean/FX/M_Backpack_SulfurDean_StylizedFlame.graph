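# Translucent, unlit, two-sided stylized flame material for SulfurDean backpack Niagara sprite and mesh particles.
# Particle color tints a two-color vertical gradient passed through an emissive adjustment function.
# Opacity combines panning noise, view-angle terms and UV gradients, then blends hard rounding with smoothstep glow, adding depth fade and an optional pickaxe mask.
# Separate branches compute view-based fade, sine vertex wobble and a clamped, UV-weighted bend driven by Niagara dynamic parameters.
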Material /BRCosmetics/Accessories/FORT_Backpacks/Backpack_M_MED_SulfurDean/FX/M_Backpack_SulfurDean_StylizedFlame
graph {
 properties {
  bAutomaticallySetUsageInEditor=false
  bUsedWithNiagaraMeshParticles=true
  bUsedWithNiagaraSprites=true
  blendMode=BLEND_Translucent
  previewMesh=/BRCosmetics/Weapons/FORT_Melee/Pickaxe_SulfurDean/FX/SM_Pickaxe_SulfurDean_Flame
  shadingModel=MSM_Unlit
  twoSided=true
  usedShadingModels=Unlit
 }
 nodes [
  Add()
  n1 = Transform(None=VectorParameter() {parameterName=ViewBasedFadeLocalAxis, defaultValue=<rgba 0 0 1 1>}.RGB) {transformSourceType=TRANSFORMSOURCE_Local}
  n2 = CameraVectorWS()
  n3 = VertexNormalWS()
  n4 = Min(A=Abs(None=DotProduct(A=@n1, B=@n2)), B=Abs(None=DotProduct(A=@n3, B=@n2)))
  n5 = ScalarParameter() {defaultValue=2, parameterName=ViewBasedFadeSensitivity, sliderMax=16}
  Clamp(None=SmoothStep(
    Min=ScalarParameter() {defaultValue=0.1, parameterName=ViewBasedFadeMin, sliderMax=2, sliderMin=-2}
    Max=ScalarParameter() {defaultValue=0.8, parameterName=ViewBasedFadeMax, sliderMax=2, sliderMin=-2}
    Value=Power(Base=@n4, Exp=@n5)
   ))
  n6 = DynamicParameter() {desc=Parameters set by Niagara, parameterIndex=1, paramNames [X,Y,Z,WPOIntensity]}
  n7 = Subtract(
   A=MaterialFunctionCall(X=@n6.X, Y=@n6.X, Z=@n6.X) {materialFunction=&MakeFloat3}.Result
   B=ParticlePositionWS()
  )
  n8 = ScalarParameter() {defaultValue=5, parameterName=Interpolate_WPO_Range}
  n9 = Multiply(A=@n8) {constB=-1}
  n10 = TextureCoordinate()
  n11 = Power(Base=OneMinus(None=ComponentMask(None=@n10) {g=true}), Exp=ScalarParameter() {defaultValue=4, parameterName=WPOBendPower})
  LinearInterpolate(
   A=Constant()
   B=LinearInterpolate(B=Clamp(None=@n7, Min=@n9, Max=@n8) {maxDefault=50, minDefault=-50}, Alpha=Multiply(A=@n11, B=@n6.WPOIntensity))
   Alpha=ScalarParameter() {defaultValue=1, parameterName=Use interpolate WPO}
  )
  n12 = ParticleColor()
  n13 = VectorParameter() {parameterName=Color_1, defaultValue=<rgba 2 0.7546 0.0417 1>}
  n14 = VectorParameter() {parameterName=Color_2, defaultValue=<rgba 2 0.2436 0.0312 1>}
  n15 = TextureCoordinate()
  n16 = OneMinus(None=ComponentMask(None=@n15) {g=true})
  n17 = ScalarParameter() {defaultValue=0.5657, parameterName=ColorGradient_PWR}
  n18 = Saturate(None=Multiply(A=Power(Base=@n16, Exp=@n17), B=ScalarParameter() {defaultValue=1.3, parameterName=ColorGradient_Mult}))
  MaterialFunctionCall(Emissive=Multiply(A=@n12.RGB, B=LinearInterpolate(A=@n13.RGB, B=@n14.RGB, Alpha=@n18))) {materialFunction=&MF_FX_EmissiveAdjustments}
  n19 = DynamicParameter() {paramNames [Offset1,Offset2,Param3,Param4]}
  n20 = Add(A=@n15, B=AppendVector(A=@n19.Offset1, B=@n19.Offset1))
  n21 = AppendVector(A=ScalarParameter() {defaultValue=5, parameterName=Tile_Mult_1}, B=ScalarParameter() {defaultValue=1.5, parameterName=Tile_Mult_2})
  n22 = Constant2Vector() {g=0.15, r=0.01}
  n23 = ScalarParameter() {defaultValue=8, parameterName=SpeedMult}
  n24 = TextureSampleParameter2D(UVs=Panner(Coordinate=Multiply(A=@n20, B=@n21), Speed=Multiply(A=@n22, B=@n23)) {speedX=-0.5}) {
   parameterName=Texture
   samplerType=SAMPLERTYPE_Masks
   texture=/BRCosmetics/Weapons/FORT_Melee/Pickaxe_SteelStomper/FX/M_Pickaxe_SteelStomper_PackedNoise
  }
  n25 = StaticComponentMaskParameter(None=@n24.RGB) {defaultB=true, parameterName=Texture_Mask}
  n26 = CameraVectorWS()
  n27 = DotProduct(A=@n26, B=Transform(None=Constant3Vector() {constant=<rgba 0 0 1 0>}) {transformSourceType=TRANSFORMSOURCE_Local})
  n28 = Abs(None=@n27)
  n29 = Abs(None=DotProduct(A=@n26, B=VertexNormalWS()))
  n30 = Power(Base=@n29) {constExponent=1.5}
  n31 = Add(A=Saturate(None=Power(Base=@n28)), B=Multiply(A=Saturate(None=@n30), B=ScalarParameter() {defaultValue=3.8874, parameterName=FresnelMult}))
  n32 = Power(Base=ComponentMask(None=@n15) {g=true}) {constExponent=0.9}
  n33 = ComponentMask(None=@n15) {g=true}
  n34 = Multiply(A=@n33, B=ScalarParameter() {parameterName=GradientAdd_Mult})
  n35 = Multiply(A=Saturate(None=@n27))
  n36 = Multiply(A=Saturate(None=Add(A=@n35, B=@n29)), B=ScalarParameter() {defaultValue=0.4, parameterName=FresnelAddMult}) {constB=0.3}
  n37 = ScalarParameter() {parameterName=Pickaxe_FresnelOpacityAdd_Mult}
  n38 = Power(Base=@n27) {constExponent=15}
  n39 = Power(Base=Subtract(A=OneMinus(None=@n27)) {constB=1.2})
  n40 = Multiply(A=Add(A=Saturate(None=@n38), B=Saturate(None=@n39))) {constB=0.7}
  n41 = Add(A=Multiply(A=Multiply(A=@n25, B=@n31), B=Saturate(None=@n32)), B=Add(A=Add(A=@n34, B=@n36), B=Multiply(A=@n37, B=@n40)))
  n42 = Multiply(A=Saturate(None=Power(Base=@n33) {constExponent=11}), B=ScalarParameter() {defaultValue=0.5666, parameterName=BottomOpacity_Add_Multiply})
  n43 = OneMinus(None=@n33)
  n44 = Power(Base=@n27) {constExponent=4}
  n45 = Multiply(A=Multiply(A=Power(Base=@n43), B=Saturate(None=@n44)), B=ScalarParameter() {parameterName=TopOpacity_Add_Multiply})
  n46 = Multiply(A=Add(A=Add(A=@n41, B=@n42), B=Saturate(None=@n45)), B=ScalarParameter() {defaultValue=0.66, parameterName=Tex_Mult}) {constB=24}
  n47 = Multiply(A=@n46)
  n48 = ScalarParameter() {parameterName=GlowMin}
  n49 = ScalarParameter() {defaultValue=1, parameterName=GlowMax}
  n50 = Saturate(None=LinearInterpolate(A=Round(None=@n47), B=SmoothStep(Min=@n48, Max=@n49, Value=@n46), Alpha=ScalarParameter() {parameterName=LerpGlow}))
  n51 = DepthFade(Opacity=@n50, FadeDistance=ScalarParameter() {defaultValue=1, parameterName=DepthFade})
  n52 = Add(A=@n40, B=@n42)
  n53 = Saturate(None=TwoSidedSign())
  Multiply(
   A=Multiply(A=@n12.A, B=ShadingPathSwitch(Default=@n51, Deferred=@n51, Forward=@n50, Mobile=@n50))
   B=LinearInterpolate(A=LinearInterpolate(B=@n52, Alpha=@n53) {constA=1, constB=0}, Alpha=ScalarParameter() {parameterName=LerpPickaxe})
  )
  n54 = ComponentMask(None=Add(A=TextureCoordinate() {uTiling=2.2})) {r=true}
  n55 = Multiply(A=Time()) {constB=-0.9}
  n56 = Multiply(A=Sine(None=Add(A=@n54, B=@n55)), B=VectorParameter() {parameterName=WPO Direction, defaultValue=<rgba -0.3105 1 0 0>}.RGB)
  n57 = Transform(None=@n56) {transformSourceType=TRANSFORMSOURCE_World, transformType=TRANSFORM_Camera}
  n58 = Transform(None=@n56) {transformSourceType=TRANSFORMSOURCE_Local}
  n59 = ScalarParameter() {defaultValue=1, parameterName=Change Transform}
  n60 = ComponentMask(None=TextureCoordinate()) {g=true}
  n61 = Sine(None=@n60) {period=2}
  Multiply(
   A=Multiply(A=LinearInterpolate(A=@n57, B=@n58, Alpha=@n59), B=ScalarParameter() {defaultValue=1.5, parameterName=WPOSine}) {constB=4}
   B=LinearInterpolate(Alpha=Saturate(None=@n61))
  )
  Multiply(A=@n60) {constB=3}
  OneMinus(None=@n60)
 ]
}
