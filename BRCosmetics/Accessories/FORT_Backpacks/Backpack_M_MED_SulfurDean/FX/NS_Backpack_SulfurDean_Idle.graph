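# Idle Niagara effect for the SulfurDean back bling: a flame with embers, eye sprites, and a glow with a flickering light.
# Mesh flames and eye sprites track skeletal mesh sockets, while embers rise from box volumes and shrink to zero size when System.IsRunning is set.
# A shared system Timeline pulses flame color and height, ember brightness, and eye size.
# Front-end flags and time-of-day scalars adjust opacity and glow.
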NiagaraSystem /BRCosmetics/Accessories/FORT_Backpacks/Backpack_M_MED_SulfurDean/FX/NS_Backpack_SulfurDean_Idle
dependencies {
 usedDataInterfaces [
  &NiagaraDataInterfaceCurve
  &NiagaraDataInterfaceColorCurve
  &NiagaraDataInterfaceArrayFloat
  &NiagaraDataInterfaceVectorCurve
  &NiagaraDataInterfaceVector2DCurve
  &NiagaraDataInterfaceSkeletalMesh
 ]
 usedDynamicInputs [
  &Vector2DFromFloat
  &SelectFloatFromArray
  &Lerp_Float
  &SimulationPosition
  &SetBoolByFloatComparison
  &VectorLength
  &SetBoolByIntComparison
 ]
 usedModules [
  &self:SystemSpawnScript.SystemScriptSource.SystemScriptGraph.NiagaraNodeAssignment_2.SetVariables_#1
  &self:SystemSpawnScript.SystemScriptSource.SystemScriptGraph.NiagaraNodeAssignment_4.SetVariables_#2
  &SystemState
  &self:SystemSpawnScript.SystemScriptSource.SystemScriptGraph.NiagaraNodeAssignment_0.SetVariables_#3
  &self:SystemSpawnScript.SystemScriptSource.SystemScriptGraph.NiagaraNodeAssignment_3.SetVariables_#4
  &Timeline
  &self:SystemSpawnScript.SystemScriptSource.SystemScriptGraph.NiagaraNodeAssignment_1.SetVariables_#5
 ]
}
emitters {
 CoreFlame {
  emitterUpdateScript [
   &EmitterState(
    Inactive Response=::NewEnumerator1
    Life Cycle Mode=::NewEnumerator1
    Loop Behavior=::NewEnumerator1
    Loop Duration Mode=::NewEnumerator1
    UseLoopDelay=false
   )
   &SpawnBurst_Instantaneous(Spawn Count=4, Use Spawn Probability=false, UseLoopCountLimit=false)
  ]
  particleSpawnScript [
   &FN_SetSpawnGroup()
   &InitializeParticle(
    Color=<rgba 0.02 0.0003 0 0.6>
    Lifetime=&RandomRangeFloat(Minimum=1, Maximum=2)
    Mesh Scale=&MergeHelper_VectorSwitch(
     0=<0.423 0.423 0.5877>
     1=<0.305 0.305 0.4796>
     2=<0.423 0.423 0.5877>
     3=<0.31 0.31 0.4796>
     Index=&SpawnGroup()
     Num=::NewEnumerator2
    )
    Mesh Scale Mode=::NewEnumerator4
    Position Offset=&MergeHelper_VectorSwitch(
     0=<0 -13.1933 -10.2658>
     1=<0 -13.1933 -9.8355>
     2=<0 -13.1933 -10.2658>
     3=?
     Index=&SpawnGroup()
     Num=::NewEnumerator1
    )
    UsePositionOffset=false
    Write Material Random=true
   ) {module=InitializeParticle001}
   &InitialMeshOrientation(Mesh Orientation Mode=::NewEnumerator2, Rotation=<0 0 0.125>, Use Rotation Vector=true)
   &DynamicMaterialParameters(
    Index 0 Param 1=0.2
    Index 0 Param 2=&MergeHelper_FloatSwitch(0=0.7, 1=0, 2=0.4, 3=?, Index=&SpawnGroup(), Num=::NewEnumerator1)
    Index 0 Param 3=0.3
    Index 0 Param 4=&MakeCustomFloatFromBool(Bool=$User.bIsFrontEndPreview, True Float=0, False Float=1)
    Index 3 Param 2=&RandomRangeFloat(Minimum=0.9, Maximum=1.1)
    Index 3 Param 3=1
    Index 3 Param 4=1
    Module.Index0Float0Write=true
    Module.Index0Float1Write=true
    Module.Index0Float2Write=true
    Module.Index0Float3Write=true
    Module.Index3Float0Write=true
    Module.Index3Float1Write=true
    Module.Index3Float2Write=true
    Module.Index3Float3Write=true
    Write Parameter Index 3=::NewEnumerator1
   )
   &CameraOffset(Camera Offset Amount=&MergeHelper_FloatSwitch(0=0, 1=0, 2=0, 3=-0.2, Index=&SpawnGroup(), Num=::NewEnumerator2)) {module=CameraOffset001}
   /SetVariables(Particles.VisibilityTag=&SpawnGroup()) {isSetParameters=true}
  ]
  particleUpdateScript [
   &ParticleState(
    Kill Particles When Lifetime Has Elapsed=false
    Let Infinitely Lived Particles Die When Emitter Deactivates=true
    Loop Particles Lifetime=true
   )
   /SetVariables(Particles.ColorGlow:LinearColor=&MultiplyLinearColor(
     Linear Color=<rgba 1 0.7446 0.2969 1>
     Scale Factor=&MakeVector4(
      X=1
      Y=1
      Z=1
      W=&Multiply_Float(
       A=&FNTimeOfDayScalar(Night Scale=0.6, Morning Scale=0.9, Evening Scale=0.8)
       B=&MergeHelper_FloatSwitch(0=0, 1=0, 2=0.07, 3=0.706, Index=&SpawnGroup(), Num=::NewEnumerator2)
      )
     )
    )) {isSetParameters=true}
   /SetVariables(Particles.Color002=&MultiplyLinearColor(
     Linear Color=&MergeHelper_ColorSwitch(
      0=&ColorFromCurve(
       ColorCurve=DataInterface() {propertyValues {
         AlphaCurve {keys [<at 0 1 RCIM_Cubic>,<at 1 1 RCIM_Cubic>]}
         BlueCurve {keys [<at 0 0.2969 RCIM_Cubic>,<at 0.3497 1.1875>,<at 1 0.2969 RCIM_Cubic>]}
         GreenCurve {keys [<at 0 0.7446 RCIM_Cubic>,<at 0.3497 2.9786>,<at 1 0.7446 RCIM_Cubic>]}
         RedCurve {keys [<at 0 1 RCIM_Cubic>,<at 0.3497 4>,<at 1 1 RCIM_Cubic>]}
        }}
       CurveIndex=$System.Timeline.Time
      )
      1=&LerpLinearColors(
       StartColor=&ColorFromCurve(
        ColorCurve=DataInterface() {propertyValues {
          AlphaCurve {keys [<at 0 1 RCIM_Cubic>,<at 1 1 RCIM_Cubic>]}
          BlueCurve {keys [<at 0 0.2969 RCIM_Cubic>,<at 0.3497 0.3266>,<at 1 0.2969 RCIM_Cubic>]}
          GreenCurve {keys [<at 0 0.7446 RCIM_Cubic>,<at 0.3497 0.8191>,<at 1 0.7446 RCIM_Cubic>]}
          RedCurve {keys [<at 0 1 RCIM_Cubic>,<at 0.3497 1.1>,<at 1 1 RCIM_Cubic>]}
         }}
        CurveIndex=$System.Timeline.Time
       )
       EndColor=<rgba 1 0.7844 0.4062 1>
       LerpFactor=&FN_FrontEndFloatModifier(Gameplay=0, Front End=1)
      )
      2=&ColorFromCurve(
       ColorCurve=DataInterface() {propertyValues {
         AlphaCurve {keys [<at 0 1 RCIM_Cubic>,<at 1 1 RCIM_Cubic>]}
         BlueCurve {keys [<at 0 0.2969 RCIM_Cubic>,<at 0.3497 0.3266>,<at 1 0.2969 RCIM_Cubic>]}
         GreenCurve {keys [<at 0 0.7446 RCIM_Cubic>,<at 0.3497 0.8191>,<at 1 0.7446 RCIM_Cubic>]}
         RedCurve {keys [<at 0 1 RCIM_Cubic>,<at 0.3497 1.1>,<at 1 1 RCIM_Cubic>]}
        }}
       CurveIndex=$System.Timeline.Time
      )
      3=?
      Index=&SpawnGroup()
      Num=::NewEnumerator1
     )
     Scale Factor=&MakeVector4(X=1, Y=1, Z=1, W=&FN_FrontEndFloatModifier(Gameplay=1, Front End=1))
    )) {isSetParameters=true}
   &InterpolateOverTime(Rate Of Change=20, Target Vector Value:NiagaraPosition=$Particles.Position, Variable Type=::NewEnumerator2)
   /SetVariables(Particles.Interpolate Dynamic=&MakeFloat4FromVectorAndFloat("Vector (RGB)"=$Particles.InterpolateOverTime.InterpolateOverTime_V3_.Moving Average, "Float (Alpha)"=0.7)) {isSetParameters=true}
   &ScaleMeshSize(Scale Factor=&VectorFromCurve(
     VectorCurve=DataInterface() {propertyValues {
       XCurve {keys [<at 0 1 RCIM_Cubic>,<at 1 1 RCIM_Cubic>]}
       YCurve {keys [<at 0 1 RCIM_Cubic>,<at 1 1 RCIM_Cubic>]}
       ZCurve {keys [<at 0 1 RCIM_Cubic>,<at 0.2 1.15 RCIM_Cubic>,<at 1 1 RCIM_Cubic>]}
      }}
     CurveIndex=$System.Timeline.Time
    )) {module=ScaleMeshSize002}
   &SkeletalMeshLocation(
    MaskSpawnGroup=false
    Mesh Sampling Type=::NewEnumerator5
    Sampled Position Offset=&MergeHelper_VectorSwitch(0=<0 0 -4.5>, 1=<0 0 -4.5>, 2=<0 0 -4.5>, 3=<0 0 -4.5>, Index=&SpawnGroup(), Num=::NewEnumerator2)
    Skeletal Mesh=$System.Skeletal Mesh
    Socket Sampling Mode=::NewEnumerator5
   ) {module=SkeletalMeshLocation001}
  ]
  rendererClasses [&NiagaraMeshRendererProperties]
  renderers [
   {rendererClass=&NiagaraMeshRendererProperties, rendererIndex=0}
   {rendererClass=&NiagaraMeshRendererProperties, rendererIndex=1}
   {rendererClass=&NiagaraMeshRendererProperties, rendererIndex=2}
   {rendererClass=&NiagaraMeshRendererProperties, rendererIndex=3}
  ]
 }
 Embers {
  emitterUpdateScript [
   &EmitterState(Inactive Response=::NewEnumerator1, Life Cycle Mode=::NewEnumerator1, Loop Duration=1, UseLoopDelay=false)
   &SpawnRate(SpawnRate=11, Use Spawn Probability=false)
  ]
  particleSpawnScript [
   &FN_SetSpawnGroup(Number of Spawn Groups=2)
   &InitializeParticle(
    Color=&RandomRangeLinearColor(Minimum=<rgba 1 1 1 1>, Maximum=<rgba 1 1 1 2>)
    Lifetime Max=1
    Lifetime Min=0.8
    Lifetime Mode=::NewEnumerator1
    Sprite Rotation Mode=::NewEnumerator2
    Sprite Size Max=&Multiply_Vector2D_ByFloat(Float=&MakeCustomFloatFromBool(Bool=$System.IsRunning, True Float=0, False Float=1), Vector2D=<11 32>)
    Sprite Size Min=&Multiply_Vector2D_ByFloat(Float=&MakeCustomFloatFromBool(Bool=$System.IsRunning, True Float=0, False Float=1), Vector2D=<6 20>)
    Sprite Size Mode=::NewEnumerator2
    UsePositionOffset=false
    Write Material Random=true
   ) {module=InitializeParticle001}
   &CameraOffset(Camera Offset Amount=1)
   &DynamicMaterialParameters(
    Index 0 Param 1=&RandomRangeFloat(Minimum=0, Maximum=1)
    Index 0 Param 2=&RandomRangeFloat(Minimum=0, Maximum=1)
    Index 0 Param 3=0.3
    Index 0 Param 4=&MakeCustomFloatFromBool(Bool=$User.bIsFrontEndPreview, True Float=0, False Float=1)
    Index 3 Param 2=&RandomRangeFloat(Minimum=0.9, Maximum=1.1)
    Index 3 Param 3=1
    Index 3 Param 4=1
    Module.Index0Float0Write=true
    Module.Index0Float1Write=true
    Module.Index0Float2Write=true
    Module.Index0Float3Write=true
    Module.Index3Float0Write=true
    Module.Index3Float1Write=true
    Module.Index3Float2Write=true
    Module.Index3Float3Write=true
    Write Parameter Index 3=::NewEnumerator1
   ) {module=DynamicMaterialParameters001}
   &ShapeLocation(
    Box Size=&MergeHelper_VectorSwitch(0=<18 9 5>, 1=<9 9 3>, 2=?, 3=?, Index=&SpawnGroup(), Num=::NewEnumerator0)
    MaskSpawnGroup=false
    Offset=<0 -12.7109 0.8179>
    Offset Coordinate Space=::Local
    Offset Mode=::NewEnumerator0
    Override Seed=false
    Shape Primitive=::NewEnumerator2
    Surface Only=false
    UseOwnerScale=true
   )
   &AddVelocity(
    Velocity=&MergeHelper_VectorSwitch(
     0=&RandomRangeVector(Minimum=<-3 -3 25>, Maximum=<3 3 40>)
     1=&RandomRangeVector(Minimum=<-13 0 34>, Maximum=<13 13 40>)
     2=?
     3=?
     Index=&SpawnGroup()
     Num=::NewEnumerator0
    )
    Velocity Speed Scale=&MergeHelper_FloatSwitch(0=1, 1=0.792, 2=?, 3=?, Index=&SpawnGroup(), Num=::NewEnumerator0)
   )
  ]
  particleUpdateScript [
   &ParticleState()
   &ScaleSpriteSize(
    Uniform Curve Scale=0.2
    Uniform Curve Sprite Scale=DataInterface() {propertyValues {Curve {keys [
        {arriveTangent=1.701, interpMode=RCIM_Cubic, time=0, value=0.6492}
        {arriveTangent=-0.6492, interpMode=RCIM_Cubic, leaveTangent=-0.6492, time=0.6431, value=1}
        <at 1 0>
       ]}}}
   )
   &ScaleColor(
    Scale Alpha=&FloatFromCurve(
     FloatCurve=DataInterface() {propertyValues {Curve {keys [
         <at 0 0 RCIM_Cubic>
         {arriveTangent=0.6135, interpMode=RCIM_Cubic, leaveTangent=0.6135, time=0.2, value=1}
         {arriveTangent=-0.1495, interpMode=RCIM_Cubic, leaveTangent=-0.1495, tangentMode=RCTM_User, time=1, value=0.6135}
        ]}}}
     Scale Curve=3
    )
    Scale RGB=&MakeVectorFromLinearColorRGB(Color=&ColorFromCurve(
      ColorCurve=DataInterface() {propertyValues {
        AlphaCurve {keys [<at 0 1 RCIM_Cubic>,<at 1 1 RCIM_Cubic>]}
        BlueCurve {keys [<at 0 0.2969 RCIM_Cubic>,<at 0.3497 1.1875>,<at 1 0.2969 RCIM_Cubic>]}
        GreenCurve {keys [<at 0 0.7446 RCIM_Cubic>,<at 0.3497 2.9786>,<at 1 0.7446 RCIM_Cubic>]}
        RedCurve {keys [<at 0 1 RCIM_Cubic>,<at 0.3497 4>,<at 1 1 RCIM_Cubic>]}
       }}
      CurveIndex=$System.Timeline.Time
     ))
    ScaleA=true
    ScaleRGB=true
   )
   &ScaleColor(
    Scale Alpha=&FN_FrontEndFloatModifier(
     Gameplay=1
     Front End=&FloatFromCurve(FloatCurve=DataInterface() {propertyValues {Curve {keys [<at 0.2 0>,<at 1 1>]}}}, CurveIndex=$Engine.System.Age)
    )
    ScaleA=true
    ScaleRGB=true
   ) {module=ScaleColor001}
   &SolveForcesAndVelocity(Clamp Velocity=false, Limit Acceleration=false)
   /SetVariables(
    Particles.GlowColor=&LerpLinearColors(
     StartColor=<rgba 2 0.2945 0.033 0.083>
     EndColor=<rgba 2 0.2945 0.033 0.054>
     LerpFactor=&FN_FrontEndFloatModifier(Gameplay=0, Front End=1)
    )
    Particles.GlowSize=&Multiply_Vector2D(A=<2 1.4>, B=$Particles.SpriteSize)
   ) {isSetParameters=true}
  ]
  rendererClasses [&NiagaraSpriteRendererProperties]
  renderers [
   {rendererClass=&NiagaraSpriteRendererProperties}
   {rendererClass=&NiagaraSpriteRendererProperties, rendererIndex=1}
  ]
 }
 Eyes {
  emitterUpdateScript [
   &EmitterState(
    Inactive Response=::NewEnumerator1
    Life Cycle Mode=::NewEnumerator1
    Loop Behavior=::NewEnumerator1
    Loop Duration Mode=::NewEnumerator1
    UseLoopDelay=false
   )
   &SpawnBurst_Instantaneous(Spawn Count=4, Use Spawn Probability=false)
  ]
  particleSpawnScript [
   &FN_SetSpawnGroup()
   &InitializeParticle(Mass Mode=::NewEnumerator1, Sprite Size=<2 2>, Sprite Size Mode=::NewEnumerator4, UsePositionOffset=false, Write Material Random=true) {module=InitializeParticle001}
   &CameraOffset(Camera Offset Amount=&MergeHelper_FloatSwitch(0=6, 1=6, 2=0.1, 3=0.1, Index=&SpawnGroup(), Num=::NewEnumerator2))
   /SetVariables(Particles.VisibilityTag=&MergeHelper_IntSwitch(0=0, 1=0, 2=1, 3=1, 4=?, 5=?, Index=&SpawnGroup(), Num=::NewEnumerator2)) {isSetParameters=true}
  ]
  particleUpdateScript [
   &ParticleState(
    Kill Particles When Lifetime Has Elapsed=false
    Let Infinitely Lived Particles Die When Emitter Deactivates=true
    Loop Particles Lifetime=true
   )
   &SolveForcesAndVelocity(Clamp Velocity=false, Limit Acceleration=false)
   &ScaleSpriteSize(
    Uniform Curve Index=$System.Timeline.Time
    Uniform Curve Scale=3
    Uniform Curve Sprite Scale=DataInterface() {propertyValues {Curve {keys [
        {arriveTangent=4.0197, interpMode=RCIM_Cubic, leaveTangent=4.0197, tangentMode=RCTM_User, time=0, value=0}
        {arriveTangent=0.5805, interpMode=RCIM_Cubic, leaveTangent=0.5805, tangentMode=RCTM_User, time=0.3662, value=1}
        {arriveTangent=-0.1188, interpMode=RCIM_Cubic, leaveTangent=-0.1188, tangentMode=RCTM_User, time=0.6283, value=0.9648}
        <at 1 0 RCIM_Cubic>
       ]}}}
   ) {module=ScaleSpriteSize001}
   &FN_Sprite_Attributes(
    Color=<rgba 1 1 1 0.559>
    Create Color Attribute=true
    Create SpriteSize Attribute=true
    SpriteSize=&Multiply_Vector2D_ByFloat(Float=1.5, Vector2D=$Particles.SpriteSize)
    Write to Module Namespace=true
   )
   &SkeletalMeshLocation(
    Bone / Socket Index=&MergeHelper_IntSwitch(0=1, 1=2, 2=1, 3=2, 4=?, 5=?, Index=&SpawnGroup(), Num=::NewEnumerator2)
    MaskSpawnGroup=false
    Mesh Sampling Type=::NewEnumerator5
    Sampled Position Offset=&MergeHelper_VectorSwitch(0=<0 0 0>, 1=<0 0 0>, 2=<0 0 0>, 3=<0 0 0>, Index=&SpawnGroup(), Num=::NewEnumerator2)
    Skeletal Mesh=$System.Skeletal Mesh
    Socket Sampling Mode=::NewEnumerator5
   ) {module=SkeletalMeshLocation002}
   &SpriteRotationRate(Rotation Rate=&MergeHelper_FloatSwitch(0=120, 1=120, 2=0, 3=0, Index=&SpawnGroup(), Num=::NewEnumerator2))
   /SetVariables(
    Particles.GlowColor=<rgba 1 1 1 0.1913>
    Particles.GlowSize=&Multiply_Vector2D_ByFloat(Float=1.2, Vector2D=<2 2>)
   ) {isSetParameters=true}
   /SetVariables(
    Particles.GlowColor2:LinearColor=<rgba 14.8261 14.8261 14.8261 1>
    Particles.GlowSize2:Vector2f=&Multiply_Vector2D_ByFloat(Float=0.3, Vector2D=<2 2>)
   ) {isSetParameters=true}
  ]
  rendererClasses [&NiagaraSpriteRendererProperties]
  renderers [
   {rendererClass=&NiagaraSpriteRendererProperties}
   {rendererClass=&NiagaraSpriteRendererProperties, rendererIndex=1}
   {rendererClass=&NiagaraSpriteRendererProperties, rendererIndex=2}
   {rendererClass=&NiagaraSpriteRendererProperties, rendererIndex=3}
  ]
 }
 Glow {
  emitterUpdateScript [
   &EmitterState(
    Inactive Response=::NewEnumerator1
    Life Cycle Mode=::NewEnumerator1
    Loop Behavior=::NewEnumerator1
    Loop Duration Mode=::NewEnumerator1
    UseLoopDelay=false
   )
   &SpawnBurst_Instantaneous(Use Spawn Probability=false, UseLoopCountLimit=false)
  ]
  particleSpawnScript [
   &InitializeParticle(
    Color=<rgba 4 1.0231 0 1>
    Position Offset=<0 -12.7109 5.0697>
    Sprite Size=<40.2852 54.4832>
    Sprite Size Mode=::NewEnumerator4
    UsePositionOffset=true
    Write Material Random=true
   )
   &CameraOffset(Camera Offset Amount=-2)
  ]
  particleUpdateScript [
   &ParticleState(Kill Particles When Lifetime Has Elapsed=false, Let Infinitely Lived Particles Die When Emitter Deactivates=true)
   &ScaleColor(
    Scale Alpha=&FN_FrontEndFloatModifier(
     Gameplay=&FNTimeOfDayScalar(Night Scale=0.14, Morning Scale=0.2, Day Scale=0.22, Evening Scale=0.18)
     Front End=0.1
    )
    ScaleA=true
    ScaleRGB=true
   )
   &SpriteFacingAndAlignment(
    Sprite Alignment=&RotateVectorByQuaternion(Vector To Rotate=<0 0.1399 1>, Quaternion=$Output.SocketLocation.SampledRotation)
    WriteAlignment=true
    WriteFacing=true
   )
   &FN_Light_Attributes(
    Color=<rgba 8 2.3483 0 1>
    Create Color Attribute=true
    Create Exponent Attribute=true
    Create Position Attribute=true
    Create Radius Attribute=true
    Exponent=2
    Position=<0 -12.7109 5.0697>
    Radius=40
    Scale Color=&Waveform("Waveform [1]"=::NewEnumerator12, "[1] Frequency"=5, "[1] Amplitude Min/Max"=<0.7 1.2>)
   )
  ]
  rendererClasses [&NiagaraSpriteRendererProperties, &NiagaraLightRendererProperties]
  renderers [
   {rendererClass=&NiagaraSpriteRendererProperties}
   {rendererClass=&NiagaraLightRendererProperties, rendererIndex=1}
  ]
 }
}
summary {
 systemName=NS_Backpack_SulfurDean_Idle
 emitters [
  {emitterName=Embers, rendererClasses [&NiagaraSpriteRendererProperties]}
  {emitterName=Glow, rendererClasses [&NiagaraSpriteRendererProperties, &NiagaraLightRendererProperties]}
  {emitterName=CoreFlame, rendererClasses [&NiagaraMeshRendererProperties]}
  {emitterName=Eyes, rendererClasses [&NiagaraSpriteRendererProperties]}
 ]
 userVariables [
  User.bIsFrontEnd {type=NiagaraBool}
  User.bIsFrontEndPreview {type=NiagaraBool}
  User.bIsSwinging {type=NiagaraBool}
  User.DebugReactivity {type=NiagaraBool}
  User.KillReactivityCount {type=NiagaraInt32}
 ]
}
