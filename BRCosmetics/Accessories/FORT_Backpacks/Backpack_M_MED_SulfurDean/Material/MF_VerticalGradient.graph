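# Material function for the SulfurDean backpack that splits a scalar input into upper and lower gradient masks around a CenterHeight parameter, grouped for WPO animation.
# The input's offset from CenterHeight, negated for the lower side, is divided by a per-side gradient distance and saturated.
# A zero-hardness SphereMask against 1 serves as a 1D falloff, with radius set by each GradientPercentage parameter.
# It outputs UpperGradient, LowerGradient and CenterHeight.
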MaterialFunction /BRCosmetics/Accessories/FORT_Backpacks/Backpack_M_MED_SulfurDean/Material/MF_VerticalGradient
graph {nodes [
  n1 = ScalarParameter() {group=WPO_Animation, parameterName=CenterHeight}
  n2 = Subtract(A=FunctionInput() {inputType=FunctionInput_Scalar}, B=@n1)
  n3 = Divide(A=@n2, B=ScalarParameter() {defaultValue=10, group=WPO_Animation, parameterName=Centergradient01})
  FunctionOutput(None=SphereMask(
    A=Saturate(None=@n3)
    B=Constant() {r=1}
    Radius=ScalarParameter() {defaultValue=0.5, group=WPO_Animation, parameterName=GradientPercentage01}
   ) {hardnessPercent=0}) {outputName=UpperGradient, sortPriority=-1}
  n4 = Divide(A=Multiply(A=@n2) {constB=-1}, B=ScalarParameter() {defaultValue=10, group=WPO_Animation, parameterName=Centergradient02})
  FunctionOutput(None=SphereMask(A=Saturate(None=@n4), B=Constant() {r=1}, Radius=ScalarParameter() {defaultValue=1, group=WPO_Animation, parameterName=GradientPercentage02}) {hardnessPercent=0}) {outputName=LowerGradient}
  FunctionOutput(None=@n1) {outputName=CenterHeight, sortPriority=-1}
 ]}
