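# Material function for the SulfurDean backpack that animates fire by adding World Position Offset to incoming material attributes.
# Pre-skinned position feeds upper/lower height gradients and pivot-relative XZ UVs for a panned noise texture, whose vertical scroll flips across the pivot.
# That noise supplies turbulence and strength for offsets directed along vectors to upper and lower target points.
# The result is texture-masked, transformed from local space and applied only where U exceeds 1.
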MaterialFunction /BRCosmetics/Accessories/FORT_Backpacks/Backpack_M_MED_SulfurDean/Material/MF_Fire_WPO
graph {nodes [
  n1 = Multiply(
   A=NamedRerouteUsage()
   B=Multiply(A=NamedRerouteUsage(), B=ScalarParameter() {defaultValue=0.5, group=WPO_Animation, parameterName=UpperTurblenceIntensity})
  )
  n2 = Subtract(A=VectorParameter() {group=WPO_Animation, parameterName=UpperTargetingPoint, defaultValue=<rgba 0 0 0 1>}.RGB, B=NamedRerouteUsage())
  n3 = LinearInterpolate(
   A=Constant() {r=1}
   B=Length(None=@n2) {is=class MaterialExpressionLength}
   Alpha=ScalarParameter() {defaultValue=1, group=WPO_Animation, parameterName=LengthBlend_Upper}
  )
  n4 = Normalize(VectorInput=@n2)
  n5 = Multiply(
   A=NamedRerouteUsage()
   B=Multiply(A=Multiply(A=@n3, B=@n4), B=ScalarParameter() {defaultValue=1, group=WPO_Animation, parameterName=WPOMulti01})
  )
  n6 = Multiply(A=NamedRerouteUsage(), B=Add(A=NamedRerouteDeclaration(None=@n5) {name=UpperWPO, nodeColor=<rgba 0.8047 0 1 1>}, B=NamedRerouteUsage()))
  n7 = TextureSampleParameter2D() {
   group=WPO_Animation
   parameterName=WPO_Mask
   samplerType=SAMPLERTYPE_Masks
   texture=/BRCosmetics/Accessories/FORT_Backpacks/Backpack_M_MED_SulfurDean/Texture/T_WPO_Mask
  }
  n8 = Add(A=Multiply(A=Add(A=@n1, B=@n6), B=@n7.RGB), B=NamedRerouteUsage())
  Add(A=@n8, B=VectorParameter() {group=WPO_Animation, parameterName=TransParentWPO_Test, defaultValue=<rgba 0 0 0 1>}.RGB)
  n9 = GetMaterialAttributes(None=FunctionInput() {inputName=Material_IN, inputType=FunctionInput_MaterialAttributes}) {attributeGetTypes [#1]}
  n10 = TextureCoordinate()
  n11 = Constant()
  n12 = Multiply(
   A=Transform(None=@n8) {transformSourceType=TRANSFORMSOURCE_Local}
   B=If(A=ComponentMask(None=@n10) {r=true}, B=Constant() {r=1}, "A > B"=Constant() {r=1}, "A == B"=@n11, "A < B"=@n11) {constB=1}
  )
  FunctionOutput(None=SetMaterialAttributes!(MaterialAttributes=@n9.MaterialAttributes, WorldPositionOffset=Add(A=@n9.WorldPositionOffset, B=@n12) {constB=0}) {attributeSetTypes [#1]})
  n13 = PreSkinnedPosition()
  n14 = NamedRerouteDeclaration(None=@n13) {name=PreskinnedPOS_Vtx, nodeColor=<rgba 0.5703 0 1 1>}
  n15 = MaterialFunctionCall!(In=ComponentMask(None=@n14) {b=true}) {materialFunction=&/MF_VerticalGradient}
  NamedRerouteDeclaration(None=@n15.UpperGradient) {name=UpperMask_Vtx, nodeColor=<rgba 1 0.8906 0 1>}
  n16 = Subtract(A=@n14, B=VectorParameter() {parameterName=Pivot, defaultValue=<rgba 0 0 0 1>}.RGB)
  n17 = Divide(A=ComponentMask(None=@n16) {b=true, r=true}, B=ScalarParameter() {defaultValue=10, group=WPO_Animation, parameterName=UV_Unit})
  n18 = ScalarParameter() {defaultValue=0.3, group=WPO_Animation, parameterName=SpeedX01}
  n19 = ScalarParameter() {defaultValue=0.54, group=WPO_Animation, parameterName=SpeedY01}
  n20 = ComponentMask(None=@n16) {b=true}
  n21 = LinearInterpolate(A=Multiply(A=@n19) {constB=-1}, B=@n19, Alpha=Saturate(None=Sign(None=@n20)))
  n22 = TextureSample(UVs=Panner(Coordinate=NamedRerouteDeclaration(None=@n17) {name=PreskinnedUV, nodeColor=<rgba 0.9688 1 0 1>}, Speed=AppendVector(A=@n18, B=@n21))) {texture=/BRCosmetics/Accessories/FORT_Backpacks/Backpack_M_MED_SulfurDean/Texture/T_FireNoise_Comp}
  n23 = Constant()
  n24 = Subtract(
   A=AppendVector(A=AppendVector(A=@n22.R, B=@n23), B=@n22.G)
   B=VectorParameter() {group=WPO_Animation, parameterName=Offset_Turblence, defaultValue=<rgba 0.5 0 0.5 1>}.RGB
  )
  NamedRerouteDeclaration(None=@n24) {name=TurblenceNoise_Pre, nodeColor=<rgba 0 0.1562 1 1>}
  NamedRerouteDeclaration(None=@n15.LowerGradient) {name=LowerMask_Vtx, nodeColor=<rgba 0.875 1 0 1>}
  n25 = Subtract(A=VectorParameter() {group=WPO_Animation, parameterName=LowerTargetingPoint, defaultValue=<rgba 0 0 0 1>}.RGB, B=NamedRerouteUsage())
  n26 = Multiply(
   A=LinearInterpolate(A=Constant() {r=1}, B=Length(None=@n25), Alpha=ScalarParameter() {defaultValue=1, group=WPO_Animation, parameterName=LengthBlend_Lower})
   B=Normalize(VectorInput=@n25)
  )
  n27 = ScalarParameter() {defaultValue=1, group=WPO_Animation, parameterName=WPOMulti02}
  NamedRerouteDeclaration(None=Multiply(A=NamedRerouteUsage(), B=Multiply(A=@n26, B=@n27))) {name=LowerWPO, nodeColor=<rgba 1 0 0.7969 1>}
  NamedRerouteDeclaration(None=VertexInterpolator(VS=@n13).PS) {name=PreskinnedPOS_PS, nodeColor=<rgba 0.5703 0 1 1>}
  NamedRerouteDeclaration(None=@n22.B) {name=WPONoise, nodeColor=<rgba 0.5 1 0 1>}
  NamedRerouteDeclaration(None=@n15.CenterHeight) {name=CenterHeight, nodeColor=<rgba 0 1 0.3672 1>}
  NamedRerouteDeclaration(None=Multiply(A=@n24, B=ScalarParameter() {defaultValue=1, group=WPO_Animation, parameterName=Turblence_Intensity})) {name=TurblenceNoise, nodeColor=<rgba 1 0.4922 0 1>}
 ]}
