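# Animation Blueprint for the SulfurDean backpack cosmetic, playing idle and a reactive animation on enemy eliminations or on timers in the front end.
# bIsInFrontEnd selects between two idle/reactive state machines, followed by an additive pack bone offset and spring-driven Anim Dynamics on dyn_lid.
# The EventGraph sets trigger flags from the elimination input event, customization load and a repeating notify delay, clearing them after evaluation.
# The TESTING_Capture state machine is unconnected.
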AnimBlueprint /BRCosmetics/Accessories/FORT_Backpacks/Backpack_M_MED_SulfurDean/Meshes/M_MED_SulfurDean_Pack_AnimBP
target_skeleton=&/M_MED_SulfurDean_Pack_Skeleton
graphs [
 AnimGraph {nodes [
   Input Pose()
   AnimGraphNode_StateMachine_1 = Idle + Elimination() {is=title AnimGraphNode_StateMachine}
   AnimGraphNode_StateMachine_2 = FE_Idle to FE_Elim() {is=title AnimGraphNode_StateMachine}
   n1 = Get bIsInFrontEnd()
   n2 = "Transform (Modify) Bone - Bone: pack"(
    ComponentPose=Local To Component(LocalPose=Blend Poses by bool(BlendPose_0=@AnimGraphNode_StateMachine_2, BlendPose_1=@AnimGraphNode_StateMachine_1, bActiveValue=@n1) {blendType=Linear})
    Translation="0,-1.3,0"
    Rotation="0,0,0"
   ) {
    translationMode=BMM_Additive
    translationSpace=BCS_BoneSpace
    alphaBoolBlend {blendOption=Linear}
    boneToModify {boneName=pack}
    translation {y=-1.3}
   }
   n3 = Anim Dynamics - Bone: dyn_lid(ComponentPose=@n2, Alpha=1, ExternalForce="0,0,0") {
    angularDampingOverride=1
    angularSpringConstant=20000
    bAngularSpring=true
    bOverrideAngularDamping=true
    bUseGravityOverride=true
    gravityScale=0.35
    simulationSpace=BoneRelative
    alphaBoolBlend {blendOption=Linear}
    boundBone {boneName=dyn_lid}
    componentAppliedLinearAccClamp=<2000 2000 2000>
    componentLinearAccScale=<0.7 0.7 0.7>
    componentLinearVelScale=<0.7 0.7 0.7>
    gravityOverride {z=-980}
    physicsBodyDefinitions [{
      boundBone {boneName=dyn_lid}
      boxExtents=<20 30 30>
      constraintSetup {angularLimitsMax=<0 15 0>, angularLimitsMin=<0 0 0>, angularTarget=<1 0 0>}
      localJointOffset=<3 0 0>
     }]
    relativeSpaceBone {boneName=pack}
   }
   Output Pose(Result=Component To Local(ComponentPose=@n3)) {layerGroup=None}
   AnimGraphNode_StateMachine_0 = TESTING_Capture() {is=title AnimGraphNode_StateMachine}
  ]}
 TESTING_Capture {
  entry=Idle
  path=AnimGraph.AnimGraphNode_StateMachine_0.TESTING_Capture
  ids=id
  nodes [
   AnimStateNode_0 = Idle()
   AnimStateNode_1 = Elim() {is=title AnimStateNode}
   AnimStateNode_2 = Idle_02() {is=title AnimStateNode}
   AnimStateNode_3 = Elim_2() {is=title AnimStateNode}
   AnimStateNode_4 = Idle_03() {is=title AnimStateNode}
   AnimStateTransitionNode_4 = AnimStateTransitionNode() {bAutomaticRuleBasedOnSequencePlayerInState=true}
   AnimStateTransitionNode_5 = AnimStateTransitionNode() {bAutomaticRuleBasedOnSequencePlayerInState=true}
   AnimStateTransitionNode_6 = AnimStateTransitionNode() {bAutomaticRuleBasedOnSequencePlayerInState=true}
   AnimStateTransitionNode_7 = AnimStateTransitionNode() {bAutomaticRuleBasedOnSequencePlayerInState=true}
   AnimStateTransitionNode_8 = AnimStateTransitionNode() {bAutomaticRuleBasedOnSequencePlayerInState=true}
  ]
  links [
   {
    dst=Idle_03
    rule=AnimGraph.AnimGraphNode_StateMachine_0.TESTING_Capture.AnimStateTransitionNode_5.Transition
    src=Elim
    via=AnimStateTransitionNode_5
   }
   {
    dst=Idle
    rule=AnimGraph.AnimGraphNode_StateMachine_0.TESTING_Capture.AnimStateTransitionNode_8.Transition
    src=Elim_2
    via=AnimStateTransitionNode_8
   }
   {
    dst=Idle_02
    rule=AnimGraph.AnimGraphNode_StateMachine_0.TESTING_Capture.AnimStateTransitionNode_4.Transition
    src=Idle
    via=AnimStateTransitionNode_4
   }
   {
    dst=Elim
    rule=AnimGraph.AnimGraphNode_StateMachine_0.TESTING_Capture.AnimStateTransitionNode_6.Transition
    src=Idle_02
    via=AnimStateTransitionNode_6
   }
   {
    dst=Elim_2
    rule=AnimGraph.AnimGraphNode_StateMachine_0.TESTING_Capture.AnimStateTransitionNode_7.Transition
    src=Idle_03
    via=AnimStateTransitionNode_7
   }
  ]
 }
 Idle {
  path=AnimGraph.AnimGraphNode_StateMachine_0.TESTING_Capture.AnimStateNode_0.Idle
  nodes [
   n1 = Sequence Player 'SulfurDean_Pack_Idle_PACK'() {bLoopAnimation=false, sequence=/BRCosmetics/Accessories/FORT_Backpacks/Backpack_M_MED_SulfurDean/Animation/SulfurDean_Pack_Idle_PACK}
   Output Animation Pose(Result=@n1)
  ]
 }
 Elim {
  path=AnimGraph.AnimGraphNode_StateMachine_0.TESTING_Capture.AnimStateNode_1.Elim
  nodes [
   n1 = Sequence Player 'SulfurDean_Pack_Reactive_PACK'() {bLoopAnimation=false, sequence=/BRCosmetics/Accessories/FORT_Backpacks/Backpack_M_MED_SulfurDean/Animation/SulfurDean_Pack_Reactive_PACK}
   Output Animation Pose(Result=@n1)
  ]
 }
 Idle_02 {
  path=AnimGraph.AnimGraphNode_StateMachine_0.TESTING_Capture.AnimStateNode_2.Idle_02
  nodes [
   n1 = Sequence Player 'SulfurDean_Pack_Idle_PACK'() {bLoopAnimation=false, sequence=&SulfurDean_Pack_Idle_PACK}
   Output Animation Pose(Result=@n1)
  ]
 }
 Elim_2 {
  path=AnimGraph.AnimGraphNode_StateMachine_0.TESTING_Capture.AnimStateNode_3.Elim_2
  nodes [
   n1 = Sequence Player 'SulfurDean_Pack_Reactive_PACK'() {bLoopAnimation=false, sequence=&SulfurDean_Pack_Reactive_PACK}
   Output Animation Pose(Result=@n1)
  ]
 }
 Idle_03 {
  path=AnimGraph.AnimGraphNode_StateMachine_0.TESTING_Capture.AnimStateNode_4.Idle_03
  nodes [
   n1 = Sequence Player 'SulfurDean_Pack_Idle_PACK'() {bLoopAnimation=false, sequence=&SulfurDean_Pack_Idle_PACK}
   Output Animation Pose(Result=@n1)
  ]
 }
 Transition {path=AnimGraph.AnimGraphNode_StateMachine_0.TESTING_Capture.AnimStateTransitionNode_4.Transition, nodes [Result()]}
 Transition {path=AnimGraph.AnimGraphNode_StateMachine_0.TESTING_Capture.AnimStateTransitionNode_5.Transition, nodes [Result()]}
 Transition {path=AnimGraph.AnimGraphNode_StateMachine_0.TESTING_Capture.AnimStateTransitionNode_6.Transition, nodes [Result()]}
 Transition {path=AnimGraph.AnimGraphNode_StateMachine_0.TESTING_Capture.AnimStateTransitionNode_7.Transition, nodes [Result()]}
 Transition {path=AnimGraph.AnimGraphNode_StateMachine_0.TESTING_Capture.AnimStateTransitionNode_8.Transition, nodes [Result()]}
 Idle + Elimination {
  entry=Idle
  path=AnimGraph.AnimGraphNode_StateMachine_1.Idle + Elimination
  ids=id
  nodes [
   AnimStateNode_0 = Idle()
   AnimStateNode_1 = Elimination() {is=title AnimStateNode}
   AnimStateTransitionNode_0 = AnimStateTransitionNode()
   AnimStateTransitionNode_1 = AnimStateTransitionNode() {bAutomaticRuleBasedOnSequencePlayerInState=true}
  ]
  links [
   {
    dst=Idle
    rule=AnimGraph.AnimGraphNode_StateMachine_1.Idle + Elimination.AnimStateTransitionNode_1.Transition
    src=Elimination
    via=AnimStateTransitionNode_1
   }
   {
    dst=Elimination
    rule=AnimGraph.AnimGraphNode_StateMachine_1.Idle + Elimination.AnimStateTransitionNode_0.Transition
    src=Idle
    via=AnimStateTransitionNode_0
   }
  ]
 }
 Idle {
  path=AnimGraph.AnimGraphNode_StateMachine_1.Idle + Elimination.AnimStateNode_0.Idle
  nodes [Output Animation Pose(Result=Sequence Player 'SulfurDean_Pack_Idle_PACK'() {sequence=&SulfurDean_Pack_Idle_PACK})]
 }
 Elimination {
  path=AnimGraph.AnimGraphNode_StateMachine_1.Idle + Elimination.AnimStateNode_1.Elimination
  nodes [
   n1 = Sequence Player 'SulfurDean_Pack_Reactive_PACK'() {bLoopAnimation=false, sequence=&SulfurDean_Pack_Reactive_PACK}
   Output Animation Pose(Result=@n1)
  ]
 }
 Transition {
  path=AnimGraph.AnimGraphNode_StateMachine_1.Idle + Elimination.AnimStateTransitionNode_0.Transition
  nodes [Result(bCanEnterTransition=Get bEnemyPlayerEliminated() -> (bEnemyPlayerEliminated:bool))]
 }
 Transition {path=AnimGraph.AnimGraphNode_StateMachine_1.Idle + Elimination.AnimStateTransitionNode_1.Transition, nodes [Result()]}
 FE_Idle to FE_Elim {
  entry=FE_Idle
  path=AnimGraph.AnimGraphNode_StateMachine_2.FE_Idle to FE_Elim
  ids=id
  nodes [
   AnimStateNode_0 = FE_Idle() {is=title AnimStateNode}
   AnimStateNode_1 = FE_Elim() {is=title AnimStateNode}
   AnimStateTransitionNode_2 = AnimStateTransitionNode()
   AnimStateTransitionNode_3 = AnimStateTransitionNode() {bAutomaticRuleBasedOnSequencePlayerInState=true}
  ]
  links [
   {
    dst=FE_Idle
    rule=AnimGraph.AnimGraphNode_StateMachine_2.FE_Idle to FE_Elim.AnimStateTransitionNode_3.Transition
    src=FE_Elim
    via=AnimStateTransitionNode_3
   }
   {
    dst=FE_Elim
    rule=AnimGraph.AnimGraphNode_StateMachine_2.FE_Idle to FE_Elim.AnimStateTransitionNode_2.Transition
    src=FE_Idle
    via=AnimStateTransitionNode_2
   }
  ]
 }
 FE_Idle {
  path=AnimGraph.AnimGraphNode_StateMachine_2.FE_Idle to FE_Elim.AnimStateNode_0.FE_Idle
  nodes [
   n1 = Sequence Player 'SulfurDean_Pack_Idle_PACK'() {bLoopAnimation=false, sequence=&SulfurDean_Pack_Idle_PACK}
   Output Animation Pose(Result=@n1)
  ]
 }
 FE_Elim {
  path=AnimGraph.AnimGraphNode_StateMachine_2.FE_Idle to FE_Elim.AnimStateNode_1.FE_Elim
  nodes [
   n1 = Sequence Player 'SulfurDean_Pack_Reactive_PACK'() {bLoopAnimation=false, sequence=&SulfurDean_Pack_Reactive_PACK}
   Output Animation Pose(Result=@n1)
  ]
 }
 Transition {
  path=AnimGraph.AnimGraphNode_StateMachine_2.FE_Idle to FE_Elim.AnimStateTransitionNode_2.Transition
  nodes [Result(bCanEnterTransition=Get Play Reactive FE Anim() -> (Play Reactive FE Anim:bool))]
 }
 Transition {path=AnimGraph.AnimGraphNode_StateMachine_2.FE_Idle to FE_Elim.AnimStateTransitionNode_3.Transition, nodes [Result()]}
 EventGraph {nodes [
   Event OnCharacterCustomizationLoaded()
   > n1 = Branch(Condition=Get bShouldDisplayFrontendReactivity() -> (bShouldDisplayFrontendReactivity:bool))
   > Do Once(Reset=@n4.then)
   > n2 = Delay(Duration=4)
   Set Play Reactive FE Anim(execute=[@n6.then, @n2.then], Play Reactive FE Anim:bool=true) -> (Output_Get:bool)
   > n3 = Cast To FortPlayerPawn(Object=TryGetFortPawn())
   > Cast To B_Athena_PartModifer_SulfurDean_Backpack!(
    execute
    Object=GetModifierForPartType!(self:Fort Player Pawn Object Reference=@n3, PartType:EFortCustomPartType Enum=Backpack) -> (ReturnValue:Custom Character Part Modifier Object Reference)
   ) -> (then, CastFailed, AsB Athena Part Modifer Sulfur Dean Backpack:B Athena Part Modifer Sulfur Dean Backpack Object Reference) {is=title K2Node_DynamicCast}
   n4 = Set Play Reactive FE Anim(execute=@n1.else, Play Reactive FE Anim=false)
   n5 = Event BlueprintProcessAnimInputEvent()
   > Branch(Condition=IsNamed(self=@n5.AnimInputEvent, InEventName=aie.EnemyPlayerEliminated))
   > Set bEnemyPlayerEliminated(bEnemyPlayerEliminated:bool=true) -> (Output_Get:bool)
   > Delay(Duration=0.2)
   > Set bEnemyPlayerEliminated(bEnemyPlayerEliminated=false)
   Event BlueprintPostEvaluateAnimation() {is=title K2Node_Event}
   > Set bEnemyPlayerEliminated(bEnemyPlayerEliminated=false)
   > Set Play Reactive FE Anim(Play Reactive FE Anim=false)
   AnimNotify_FE_ReactiveAnim() {is=title K2Node_Event}
   > Set Play Reactive FE Anim(Play Reactive FE Anim=false)
   > n6 = Delay(Duration=12)
   TryGetPawnOwner()
   Event BlueprintUpdateAnimation()
  ]}
]
variables [
 bEnemyPlayerEliminated {type=boolean}
 Play Reactive FE Anim {type=boolean}
]
